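# Unlit masked material for a glitchy, flickering hologram look on Aventurine third-party branding, flagged for Nanite and Niagara mesh particles.
# Blends an overbright, particle-colored logo texture with a screen-space panning hologram texture, then adds Fresnel.
# A fast cosine flickers that layer, which is lerped with the SceneTint-colored texture into MF_FX_EmissiveAdjustments.
# The declared DynamicParameter channels and cosine reroute feed nothing in the graph.
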Material /DealershipSharedLibrary/Assets/ThirdPartyBranding/Aventurine/M_DLR_AventurineGlitch
graph {
 properties {bUsedWithNanite=true, bUsedWithNiagaraMeshParticles=true, blendMode=BLEND_Masked, shadingModel=MSM_Unlit, usedShadingModels=Unlit}
 nodes [
  DynamicParameter() {paramNames [Vibration, TransitionPan, TransitionTexture, Param4]}
  n1 = TextureSampleParameter2D() {parameterName=Texture, texture=&/T_Aventurine}
  n2 = VectorParameter() {parameterName=SceneTint, defaultValue=<rgba 0 7.3229 10 1>}
  n3 = Cosine(None=Time()) {period=0.1}
  n4 = Saturate(None=Add(A=Constant() {r=1.85}, B=@n3))
  n5 = Multiply(A=Saturate(None=@n1.RGB), B=Constant() {r=100})
  MaterialExpressionMeshEmitterVertexColor_0 = ParticleColor()
  n6 = Constant() {r=3}
  n7 = ScreenPosition()
  n8 = TextureSample(UVs=Panner(Coordinate=ComponentMask(None=@n7.ViewportUV) {g=true, r=true}) {speedY=0.05}) {texture=&/T_Hologram}
  n9 = Add(
   A=LinearInterpolate(A=Multiply(A=@n5, B=@MaterialExpressionMeshEmitterVertexColor_0.RGB), B=Multiply(A=@n6, B=@n8.R)) {constAlpha=0.8}
   B=Multiply(A=Fresnel() {baseReflectFraction=0.005, exponent=1}) {constB=2}
  ) {constB=0}
  MaterialFunctionCall(Emissive=LinearInterpolate(A=Multiply(A=@n1.RGB, B=@n2.RGB) {constA=1}, B=Multiply(A=@n4, B=@n9)) {constAlpha=0.6}) {materialFunction=&MF_FX_EmissiveAdjustments}
  NamedRerouteDeclaration(None=@n3) {name=cosine, nodeColor=<rgba 0 1 0.5547 1>}
 ]
}
