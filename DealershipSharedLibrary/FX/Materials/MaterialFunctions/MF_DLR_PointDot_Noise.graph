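# Material function that turns mesh quads into camera-facing dot particles, similar to FakeParticles, outputting a vertex offset and an alpha.
# Pivots packed in UV1 and UV2.x anchor billboards sized by UV2.y plus Scale, with a scaled pivot vector spreading them.
# Sine and cosine wobble, phased by vertex color, UV2.y and the NoiseMovementSpeed input, adds movement.
# Alpha derives from UV1.y and Lifetime, optionally cut to a procedural circle.
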MaterialFunction /DealershipSharedLibrary/FX/Materials/MaterialFunctions/MF_DLR_PointDot_Noise
graph {
 properties {bExposeToLibrary=true, description="PointDot-Similar to FakeParticles, with scale and noise control"}
 nodes [
  n1 = AppendVector(A=TextureCoordinate() {coordinateIndex=1}, B=ComponentMask(None=TextureCoordinate() {coordinateIndex=2}) {r=true})
  n2 = Constant() {r=1}
  n3 = ComponentMask(None=TransformPosition(None=AppendVector(A=@n1, B=@n2)) {transformType=TRANSFORMPOSSOURCE_World}) {b=true, g=true, r=true}
  n4 = Normalize(VectorInput=Subtract(A=CameraPositionWS(), B=@n3)) {desc=Vec1}
  n5 = CrossProduct(A=@n4, B=Normalize(VectorInput=Constant3Vector() {constant=<rgba 0 0 1 0>}))
  n6 = Normalize(VectorInput=CrossProduct(A=@n5, B=@n4))
  n7 = Add(
   A=ComponentMask(None=TextureCoordinate() {coordinateIndex=2}) {g=true}
   B=FunctionInput(Preview=Constant() {r=2}) {inputName=Scale, inputType=FunctionInput_Scalar}
  )
  n8 = Multiply(A=@n7, B=Constant() {r=-1})
  n9 = ConstantBiasScale(None=TextureCoordinate()) {bias=-0.5, scale=2}
  n10 = Normalize(VectorInput=@n5)
  n11 = Add(
   A=Multiply(A=Multiply(A=@n6, B=@n8), B=ComponentMask(None=@n9) {r=true})
   B=Multiply(A=Multiply(A=@n10, B=@n7), B=ComponentMask(None=@n9) {g=true})
  )
  n12 = WorldPosition()
  n13 = AppendVector(A=TextureCoordinate() {coordinateIndex=1}, B=ComponentMask(None=TextureCoordinate() {coordinateIndex=2}) {r=true})
  n14 = If(
   A=FunctionInput(Preview=Constant()) {inputName=SphereMaskLocalToWorldSpace, inputType=FunctionInput_Scalar, previewValue=<0 0 0 0>}
   "A > B"=Transform(None=@n13) {transformSourceType=TRANSFORMSOURCE_Local}
   "A == B"=@n13
   "A < B"=@n13
  )
  n15 = FunctionInput(Preview=Constant() {r=4}) {inputName=Mesh Scale, inputType=FunctionInput_Scalar}
  n16 = VertexColor()
  n17 = TextureCoordinate() {coordinateIndex=2}
  n18 = Constant() {r=1}
  n19 = Add(
   A=Multiply(A=ComponentMask(None=@n17) {g=true}) {constB=0.05}
   B=ComponentMask(None=FunctionInput(Preview=@n18) {inputName=NoiseMovementSpeed, inputType=FunctionInput_Scalar}) {r=true}
  )
  n20 = Sine(None=Add(A=@n19, B=@n16.R)) {period=0.75}
  n21 = Cosine(None=Add(A=@n16.G, B=@n19))
  n22 = Sine(None=Add(A=@n16.B, B=@n19)) {period=1.2}
  n23 = AppendVector(A=AppendVector(A=@n20, B=@n21), B=@n22)
  n24 = Multiply(A=Constant() {r=-1}, B=@n23)
  n25 = Multiply(A=Multiply(A=@n20, B=@n21), B=@n22)
  n26 = FunctionInput(Preview=Constant() {r=1}) {inputName=NoiseAmount, inputType=FunctionInput_Scalar}
  n27 = AppendVector(A=@n26, B=@n26)
  n28 = ConstantBiasScale(None=Multiply(A=LinearInterpolate(A=@n24, B=@n23, Alpha=@n25), B=AppendVector(A=@n27, B=@n26)))
  n29 = FunctionInput(Preview=Constant() {r=4}) {inputName=NoiseScale, inputType=FunctionInput_Scalar}
  n30 = Add(A=Add(A=Subtract(A=@n11, B=@n12.XYZ), B=@n3), B=Add(A=Multiply(A=@n14, B=@n15) {constB=4}, B=Multiply(A=@n28, B=@n29) {constB=4}))
  FunctionOutput(None=@n30) {outputName=WorldPosition}
  n31 = ComponentMask(None=TextureCoordinate() {coordinateIndex=1}) {g=true}
  n32 = FunctionInput(Preview=Constant() {r=1}) {inputName="Lifetime(Alpha)", inputType=FunctionInput_Scalar}
  n33 = OneMinus(None=@n32)
  n34 = Saturate(None=Add(A=Subtract(A=@n31, B=@n33), B=@n32))
  n35 = StaticBool()
  n36 = Add(A=TextureCoordinate(), B=Constant() {r=-0.5})
  n37 = OneMinus(None=DotProduct(A=@n36, B=@n36))
  n38 = Constant() {r=0.75}
  n39 = Saturate(None=Ceil(None=Subtract(A=@n37, B=@n38)))
  FunctionOutput(None=StaticSwitch(True=Multiply(A=@n39, B=@n34), False=@n34, Value=FunctionInput(Preview=@n35) {inputName=CircleMaskAlpha, inputType=FunctionInput_StaticBool})) {outputName=Alpha}
  TransformPosition(None=@n30) {transformSourceType=TRANSFORMPOSSOURCE_World}
  Transform(None=@n30) {transformSourceType=TRANSFORMSOURCE_Local}
 ]
}
