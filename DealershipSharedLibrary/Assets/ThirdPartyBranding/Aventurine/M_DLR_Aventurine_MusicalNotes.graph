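# Translucent unlit material for Niagara mesh particles, named for musical notes in the Aventurine dealership branding.
# Vectors packed in UV channels 2 and 3, plus vertex color, build a per-vertex sine noise.
# MF_DLR_PointDot_Noise supplies world position offset, and UV green channels feed saturated masks for coordinates and alpha.
# Particle color and dynamic parameters expose noise, fade and end-transition controls, alongside an emissive adjustment function.
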Material /DealershipSharedLibrary/Assets/ThirdPartyBranding/Aventurine/M_DLR_Aventurine_MusicalNotes
graph {
 properties {bUsedWithNanite=true, bUsedWithNiagaraMeshParticles=true, blendMode=BLEND_Translucent, shadingModel=MSM_Unlit, usedShadingModels=Unlit}
 nodes [
  Add(A=NamedRerouteUsage(), B=NamedRerouteUsage())
  MaterialFunctionCall(Emissive=NamedRerouteUsage()) {materialFunction=&MF_FX_EmissiveAdjustments}
  n1 = TextureCoordinate() {coordinateIndex=2}
  n2 = Transform(None=AppendVector(A=TextureCoordinate() {coordinateIndex=3}, B=ComponentMask(None=@n1) {r=true})) {transformSourceType=TRANSFORMSOURCE_Local}
  n3 = VertexColor()
  n4 = Sine(None=Add(A=Divide(A=@n2) {constB=64}, B=Multiply(A=@n3.G) {constB=0.8}))
  n5 = LinearInterpolate(Alpha=@n3.R) {constA=-128, constB=455}
  NamedRerouteDeclaration(None=Multiply(A=Multiply(A=@n4, B=@n5), B=NamedRerouteUsage())) {name=Noise, nodeColor=<rgba 0 0.9907 1 1>}
  n6 = OneMinus(None=NamedRerouteUsage())
  n7 = NamedRerouteUsage()
  n8 = NamedRerouteUsage()
  n9 = NamedRerouteUsage()
  NamedRerouteDeclaration(None=MaterialFunctionCall!(
    NoiseAmount=Multiply(A=@n6, B=@n7)
    CircleMaskAlpha=StaticBool()
    "Lifetime(Alpha)"=NamedRerouteUsage()
    Scale=Multiply(A=@n8, B=@n9)
    Mesh Scale=ScalarParameter() {defaultValue=1, parameterName=MeshScale}
    NoiseScale=ScalarParameter() {defaultValue=1, parameterName=NoiseScale}
    SphereMaskLocalToWorldSpace=ScalarParameter() {parameterName="SphereMaskLocalToWorldSpace (0 or 1)"}
    NoiseMovementSpeed=Constant()
   ) {materialFunction=/DealershipSharedLibrary/FX/Materials/MaterialFunctions/MF_DLR_PointDot_Noise}.WorldPosition) {name=WPO, nodeColor=<rgba 0.0104 0.0831 1 1>}
  MaterialExpressionMeshEmitterVertexColor_2 = ParticleColor()
  NamedRerouteDeclaration(None=@MaterialExpressionMeshEmitterVertexColor_2.RGB) {name=Particle Color, nodeColor=<rgba 0 0.7187 1 1>}
  NamedRerouteDeclaration(None=@MaterialExpressionMeshEmitterVertexColor_2.A) {name=ParticleAlpha, nodeColor=<rgba 0.9844 0.1743 0.1793 1>}
  n10 = NamedRerouteUsage()
  n11 = TextureCoordinate() {coordinateIndex=3}
  n12 = Subtract(A=NamedRerouteUsage(), B=OneMinus(None=ComponentMask(None=@n11) {g=true}))
  NamedRerouteDeclaration(None=Saturate(None=Multiply(A=@n10, B=@n12))) {name=Coords, nodeColor=<rgba 1 0 0.3984 1>}
  MaterialExpressionMeshEmitterDynamicParameter_2 = DynamicParameter() {defaultValue=<rgba 115.114 3.412 1 0>, paramNames [NoiseAmount, Fade, EndTransition, Noise Mult]}
  NamedRerouteDeclaration(None=@MaterialExpressionMeshEmitterDynamicParameter_2.NoiseAmount) {name=DP NoiseAmount, nodeColor=<rgba 1 0 0.2812 1>}
  NamedRerouteDeclaration(None=@MaterialExpressionMeshEmitterDynamicParameter_2.Fade) {name=DP Fade, nodeColor=<rgba 1 0 0.2812 1>}
  NamedRerouteDeclaration(None=@MaterialExpressionMeshEmitterDynamicParameter_2.EndTransition) {name=DP EndTransition, nodeColor=<rgba 1 0 0.2812 1>}
  NamedRerouteDeclaration(None=@MaterialExpressionMeshEmitterDynamicParameter_2.Noise Mult) {name=DP NoiseMulti, nodeColor=<rgba 1 0 0.2812 1>}
  n13 = NamedRerouteUsage()
  n14 = Subtract(A=ComponentMask(None=TextureCoordinate() {coordinateIndex=2}) {g=true}, B=OneMinus(None=@n13))
  NamedRerouteDeclaration(None=Saturate(None=Add(A=@n14, B=@n13))) {name=Alpha, nodeColor=<rgba 0.5938 0.6112 1 1>}
  NamedRerouteUsage()
 ]
}
